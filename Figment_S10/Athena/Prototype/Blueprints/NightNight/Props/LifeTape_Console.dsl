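; Interactable LifeTape console prop for the Season 10 NightNight quests that plays audio dialog and can unlock a character skin variant.
; On interact, it checks three quests; if incomplete, clients step backwards through async-loaded audio lines.
; If complete, it sends a hidden-quest stat (tag lost in extraction), switches the meteorman remix material variant server-side unless that quest is done, and plays final audio.
; A developer test event restarts the dialog sequence.
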(event Custom|Check%20Quest%20Completion
  (bind _playercontroller (Variables|Default|GetPlayerController))
  (bind _returnvalue (Quests|GetQuestManager _playercontroller "Invalid"))
  (bind _playercontroller_2 (Variables|Default|GetPlayerController))
  (bind _returnvalue_2 (Quests|GetQuestManager _playercontroller_2 "Invalid"))
  (if (Quest|HasCompletedQuest _returnvalue (Variables|Default|GetQuest_01) :!class /Script/FortniteGame.FortQuestManager)
    (if (Quest|HasCompletedQuest _returnvalue (Variables|Default|GetQuest_02) :!class /Script/FortniteGame.FortQuestManager)
      (if (Quest|HasCompletedQuest _returnvalue (Variables|Default|GetQuest_03) :!class /Script/FortniteGame.FortQuestManager)
        (Variables|Default|SetWearing%20Skin (GameplayTags|TagContainer|CompareTagAssetInterfacetoOtherTags (Variables|Default|GetT_Meteorskin) (Variables|Default|GetInteractingPawn)))
        (Variables|Default|SetIs%20Complete true)
        ; S10 Extraction - Missing Tag - Athena.Quests.NightNight.Hidden
        (Quest|SendComplexCustomStatEvent _returnvalue_2 0 "(GameplayTags=)" "(GameplayTags=)" 1)
        (if (Quest|HasCompletedQuest _returnvalue_2 0 :!class /Script/FortniteGame.FortQuestManager)
          (Variables|Default|SetHiddenQuestCompleted true)
          (else
            (Variables|Default|SetHiddenQuestCompleted false)))
        (else
          _))
      (else
        _))
    (Variables|Default|SetIs%20Complete false)
    (else
      _)))

(event Custom|HandleCharacterVariantTransition
  (params (TransitioningPawn :type Fort%20Pawn%20Object%20Reference))
  ;[1 1 1] Play Skin Update FX
  (Utilities|IsValid TransitioningPawn
    (:Is%20Valid
      (Utilities|IsValid (Variables|/Script/Engine.Character|GetMesh TransitioningPawn)
        (:Is%20Valid
          (if (Variables|Default|GetHiddenQuestCompleted)
            (Networking|SwitchHasAuthority
              (:Remote
                _))
            (else
              (Networking|SwitchHasAuthority
                (:Authority
                  (Utilities|FlowControl|Delay 0.5)
                  (Class|FortPlayerController|ChangeActiveVariantforCosmeticItem (Variables|Default|GetPlayerController) "AthenaCharacter:cid_540_athena_commando_m_meteormanremix" "(TagName=\"Cosmetics.Variant.Channel.Material\")" "(TagName=\"Cosmetics.Variant.Property.Mat2\")" 0))
                (:Remote
                  (Utilities|FlowControl|Sequence
                    (:then_0
                      (Utilities|FlowControl|DoOnce false
                        (:Completed
                          (if (Variables|Default|GetWearing%20Skin)
                            _))))
                    (:then_1
                      _))))))
          (Audio|Components|Audio|FadeOut (Variables|Default|GetAudio_Key_Float) 0.25 0.0 "Linear" :!class /Script/Engine.AudioComponent)
          ;; SetFinalAudio
          (Variables|Default|SetAudioLineSelection 6)
          (bind _object (Utilities|AsyncLoadAsset (Utilities|Array|Get(acopy) (Variables|Default|GetAudioLine) (Variables|Default|GetAudioLineSelection)))
            (:Completed
              (bind _assound_base (Utilities|Casting|CastToSoundBase _object))
              (Audio|Components|Audio|SetSound (Variables|Default|GetAudio) _assound_base)))
          (Utilities|FlowControl|Delay 0.2)
          (Audio|Components|Audio|Play (Variables|Default|GetAudio) 0.0))))))

(event Custom|LocalEvents
  (params (InputPin :type Fort%20Pawn%20Object%20Reference))
  ;[1 1 1] Plays audio locally
  (bind _output_get (Variables|Default|SetInteractingPawn InputPin))
  (bind _output_get_2 (Variables|Default|SetPlayerController (Fortnite|GetFortPlayerControllerfromActor _output_get)))
  (Utilities|IsValid _output_get_2
    (:Is%20Valid
      (CallFunction|CheckQuestCompletion)
      (if (Variables|Default|GetIs%20Complete)
        (CallFunction|HandleCharacterVariantTransition (Variables|Default|GetInteractingPawn))
        (else
          (Networking|SwitchHasAuthority
            (:Remote
              ;; SetDifferentAudio
              (if (<= (Variables|Default|GetAudioLineSelection) -1)
                (Variables|Default|SetAudioLineSelection 5)
                (bind _object (Utilities|AsyncLoadAsset (Utilities|Array|Get(acopy) (Variables|Default|GetAudioLine) (Variables|Default|GetAudioLineSelection)))
                  (:Completed
                    (bind _assound_base (Utilities|Casting|CastToSoundBase _object))
                    (Audio|Components|Audio|SetSound (Variables|Default|GetAudio) _assound_base)
                    (Math|Integer|DecrementInt (Variables|Default|GetAudioLineSelection)
                      (:%20%20
                        _))))
                (else
                  (bind _object_2 (Utilities|AsyncLoadAsset (Utilities|Array|Get(acopy) (Variables|Default|GetAudioLine) (Variables|Default|GetAudioLineSelection)))
                    (:Completed
                      (bind _assound_base_2 (Utilities|Casting|CastToSoundBase _object_2))
                      (Audio|Components|Audio|SetSound (Variables|Default|GetAudio) _assound_base_2)
                      (Math|Integer|DecrementInt (Variables|Default|GetAudioLineSelection)
                        (:%20%20
                          _))))))
              (Development|PrintString (Utilities|GetDisplayName (Variables|Default|GetAudio)))
              (Audio|Components|Audio|Play (Variables|Default|GetAudio) 0.0))))))))

(event Interaction|EventBlueprintOnInteract
  (params (InteractingPawn :type Fort%20Pawn%20Object%20Reference) (InteractionBeingAttempted :type EInteractionBeingAttempted%20Enum :default "FirstInteraction"))
  ;[1 1 1] Plays local events
  (CallFunction|LocalEvents InteractingPawn))

(event Custom|PlayLifeTapFullDialog
  ;[1 1 1] Developer Test
  ; ke * PlayLifeTapFullDialog
  (Variables|Default|SetAudioLineSelection (- (Utilities|Array|Length (Variables|Default|GetAudioLine)) 1))
  ;[1 1 1] Plays audio locally
  ;; SetDifferentAudio
  (if (<= (Variables|Default|GetAudioLineSelection) -1)
    (Variables|Default|SetAudioLineSelection 5)
    (bind _object (Utilities|AsyncLoadAsset (Utilities|Array|Get(acopy) (Variables|Default|GetAudioLine) (Variables|Default|GetAudioLineSelection)))
      (:Completed
        (bind _assound_base (Utilities|Casting|CastToSoundBase _object))
        (Audio|Components|Audio|SetSound (Variables|Default|GetAudio) _assound_base)
        (Math|Integer|DecrementInt (Variables|Default|GetAudioLineSelection)
          (:%20%20
            _))))
    (else
      (bind _object_2 (Utilities|AsyncLoadAsset (Utilities|Array|Get(acopy) (Variables|Default|GetAudioLine) (Variables|Default|GetAudioLineSelection)))
        (:Completed
          (bind _assound_base_2 (Utilities|Casting|CastToSoundBase _object_2))
          (Audio|Components|Audio|SetSound (Variables|Default|GetAudio) _assound_base_2)
          (Math|Integer|DecrementInt (Variables|Default|GetAudioLineSelection)
            (:%20%20
              _))))))
  (Development|PrintString (Utilities|GetDisplayName (Variables|Default|GetAudio)))
  (Audio|Components|Audio|Play (Variables|Default|GetAudio) 0.0))

(event EventBeginPlay
  ;[1 1 1] Make Prop Visible and Async load audio
  (|CallParentFunction))

(fn BlueprintCanInteract
  (params (InteractingPawn :type Fort%20Pawn%20Object%20Reference) (InteractionBeingAttempted :type EInteractionBeingAttempted%20Enum :default "FirstInteraction") (InteractionType :type TInteractionType%20Enum :default "IT_NoInteraction"))
  (return (Variables|Default|GetInteract)))

(fn BlueprintGetInteractionString
  (params (InteractingPawn :type Fort%20Pawn%20Object%20Reference) (InteractionBeingAttempted :type EInteractionBeingAttempted%20Enum :default "FirstInteraction"))
  (return (Variables|Default|GetInteractionString)))

(fn GetCustomTint
  (return 0))

(fn GetIconPlacement
  (params (SelfActor :type Actor%20Object%20Reference) (ViewingActor :type Actor%20Object%20Reference))
  (return (Math|Transform|TransformLocation (Transformation|GetActorTransform SelfActor) (Variables|Default|GetInteractLocationOffset)) "100.000000,100.000000,100.000000"))

(fn OnRep_Interact)

(fn ConstructionScript)
