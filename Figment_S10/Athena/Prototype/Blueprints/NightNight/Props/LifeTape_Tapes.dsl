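; Collectible tape quest prop for Figment's NightNight prototype that plays an audio line when successfully interacted with, then hides itself locally.
; The success event totals completed objectives across its quests to select a line, async-loads that sound and assigns it to the audio component.
; Visibility replication reveals the tapes with query-only overlap collision, and clients check quest completion.
; Invalid quests hide the prop, except in replays or forced debug mode.
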(event EventPlaySuccessFX
  (bind _audio (Variables|Default|GetAudio))
  ;[1 1 1] Plays audio locally
  (|CallParentFunction)
  (Development|PrintString "Hello")
  ;; GetQuestCompletionAndSetAudio
  (for _array_element (Variables|Default|GetQuests)
    (bind _returnvalue (Quest|GetNumCompletedObjectives (Quests|GetQuestManager (Fortnite|GetFortPlayerControllerfromActor (Variables|Default|GetInteractingPawn)) "Invalid") _array_element))
    (Variables|Default|SetAudioLineSelection (+ _returnvalue (Variables|Default|GetAudioLineSelection))))
  (bind _result (Math|Integer|DecrementInt (Variables|Default|GetAudioLineSelection))
    (:%20%20
      (Development|PrintString (Utilities|String|ToString(Integer) _result))
      (Utilities|FlowControl|Sequence
        (:then_0
          (if (>= (Variables|Default|GetAudioLineSelection) 6)
            (Variables|Default|SetAudioLineSelection 0)))
        (:then_1
          (bind _object (Utilities|AsyncLoadAsset (Utilities|Array|Get(acopy) (Variables|Default|GetAudioLine_0) (Variables|Default|GetAudioLineSelection)))
            (:Completed
              (bind _assound_base (Utilities|Casting|CastToSoundBase _object))
              (Audio|Components|Audio|SetSound (Variables|Default|GetAudio) _assound_base)
              (Development|PrintString (Utilities|GetDisplayName _audio))))))))
  (Audio|Components|Audio|Play _audio 0.0)
  (Rendering|SetActorHiddenInGame true)
  (Variables|Default|SetisLocallyValid false))

(event EventBeginPlay
  ;[1 1 1] Make Prop Visible and Async load audio
  (if (Variables|Default|GetForceDebugOn)
    (Variables|Default|SetisLocallyValid true)
    (else
      (|CallParentFunction))))

(event EventQuestInitializedAndValid)

(event EventQuestInitializedandNotValid
  (if (Variables|Default|GetForceDebugOn)
    _
    (elif (Replay|GetIsPlayingReplay)
      _
      (else
        (Rendering|SetActorHiddenInGame true)))))

(fn BlueprintGetInteractionString
  (params (InteractingPawn :type Fort%20Pawn%20Object%20Reference) (InteractionBeingAttempted :type EInteractionBeingAttempted%20Enum :default "FirstInteraction"))
  (return (Variables|Default|GetInteractionString)))

(fn CheckForQuestComplete
  (Networking|SwitchHasAuthority
    (:Remote
      (bind (_playercontrollerout _completedquestobjective) (Quest|HasPlayerCompletedQuestObjective (Fortnite|GetFirstLocalFortPlayerController true) (Variables|QuestStuff|GetQuestItem) (Variables|Default|GetQuestBackendObjectiveName)))
      (return _completedquestobjective))))

(fn GetCustomTint
  (return 0))

(fn OnRep_Interact)

(fn OnRep_Visible
  (bind _tapes (Variables|Default|GetTapes))
  (Development|SetHiddeninGame _tapes false false)
  (Collision|SetCollisionEnabled _tapes "QueryOnly")
  (Collision|SetCollisionResponsetoAllChannels _tapes "ECR_Overlap"))

(fn ConstructionScript)
